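# Editor report for the Generic Asset Review Tool that flags why an actor or static mesh component may not bake into HLOD.
# RunFilterLogicOnObject routes actors to RunActorLogic and components to RunComponentLogic, then returns colored report data marked Errors or Possible Issues.
# The actor check tests HLOD inclusion and hidden state, gathers components with embedded Python, and errors if none pass.
# Component checks cover mobility, HLOD exclusion, visibility and missing mesh, accumulating issue text.
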EditorUtilityBlueprint /Game/NeverCook/EditorUtilities/GenericAssetReviewTool/BP/ReportExamples/Actor/Report_ActorInHLOD_Base
graphs [
 AppendErrorRowDelete {nodes [
   n1 = AppendErrorRowDelete!() -> (then, ErrorIn:String, NewError:String) {is=title K2Node_FunctionEntry}
   n2 = IsEmpty(InString=@n1)
   > Return Node!(execute, ReturnValue:String=Append!(A=@n1, B=SelectString(B="\r\n", bPickA=@n2), C:String=@n1.NewError) -> (ReturnValue))
  ]}
 EventGraph {links [], nodes []}
 GetCompName {nodes [
   n1 = GetCompName!() -> (then, Object:Object Reference) {is=title K2Node_FunctionEntry}
   n2 = GetDisplayName(Object=@n1)
   n3 = Split(SourceString=@n2, InStr=" ", SearchDir=FromStart)
   > n4 = Branch(Condition=@n3.ReturnValue)
   n5 = Split(SourceString=@n3, InStr=., SearchDir=FromStart)
   > n6 = Branch(Condition=@n5.ReturnValue)
   > Return Node(ReturnValue=@n5.RightS)
   Return Node(execute=@n6.else, ReturnValue=@n3)
   Return Node(execute=@n4.else, ReturnValue=@n2)
  ]}
 InterlAddError {nodes [
   n1 = InterlAddError!() -> (then, ErrorIn:String, NewError:String) {is=title K2Node_FunctionEntry}
   n2 = IsEmpty(InString=@n1)
   > Return Node(ReturnValue=Append(A=@n1, B=SelectString(B="\r\n", bPickA=@n2), C=@n1.NewError))
  ]}
 RunActorLogic {nodes [
   n1 = RunActorLogic!() -> (then, Actor:Actor Object Reference) {is=title K2Node_FunctionEntry}
   > Set l_actor(l_actor:Actor Object Reference=@n1) -> (Output_Get:Actor Object Reference)
   > n2 = Sequence!(execute) -> (then_0, then_1, then_2:Exec, then_3:Exec)
   > n3 = GetEditorProperty!(execute:Exec, Object:Object Reference=Get l_actor() -> (l_actor:Actor Object Reference), PropertyName:Name=bEnableAutoLODGeneration) -> (then:Exec, PropertyValue:bool, ReturnValue:bool)
   > n4 = Branch(Condition=@n3)
   Set l_error(execute=@n4.else, l_error:String=Append!(A="🔶", B="Actor has IncludeInHLOD=False so nothing will bake") -> (ReturnValue)) -> (Output_Get:String)
   > Set l_isError(l_isError:bool=true) -> (Output_Get:bool)
   Branch(execute=@n2.then_1, Condition=Get bHidden!(self:Actor Object Reference=Get l_actor()) -> (bHidden:bool))
   > Set l_error(l_error=Append(A="🔶", B="Actor has HiddenInGame=True so nothing will bake"))
   > Set l_isError(l_isError=true)
   Set l_hasFoundOneCompErrror(execute=@n2.then_2, l_hasFoundOneCompErrror:bool=false) -> (Output_Get:bool)
   > n5 = Execute Python Script!(
    execute
    PythonScript="import unreal\r\n\r\nit = unreal.ObjectIterator(unreal.StaticMeshComponent)\r\nreturn_array= []\r\nfor x in it:\r\n    actor = x.get_outer()\r\n    if actor == in_actor:\r\n        if \"TRASH\" not in str(x):\r\n            return_array.append(x)\r\n"
    in_actor:Actor Object Reference=Get l_actor()
   ) -> (then, ReturnValue, return_array:Array of Static Mesh Component Object References) {inputs [in_actor], outputs [return_array]}
   > n6 = Set l_comps(l_comps:Array of Static Mesh Component Object References=@n5.return_array) -> (Output_Get:Array of Static Mesh Component Object References)
   > n7 = For Each Loop(Array:Array of Static Mesh Component Object References=@n6) -> (Array Element:Static Mesh Component Object Reference)
   n8 = Branch(execute=@n7.Completed, Condition=Get l_atLeastOneComponent() -> (l_atLeastOneComponent:bool))
   n9 = InterlAddError!(
    execute=@n8.else
    self
    ErrorIn:String=Get l_error() -> (l_error:String)
    NewError:String=Append(A="🔶", B=Actor has no mesh components set to be in HLOD so nothing will bake in.)
   ) -> (then, ReturnValue:String) {is=title K2Node_CallFunction}
   > Set l_error(l_error=@n9)
   > Set l_isError(l_isError=true)
   n10 = RunComponentLogic!(execute=@n7.LoopBody, self, Comp:Static Mesh Component Object Reference=@n7) -> (then, Success?:bool, Error:String)
   > n11 = Branch(Condition=@n10)
   > Set l_atLeastOneComponent(l_atLeastOneComponent:bool=true) -> (Output_Get:bool)
   n12 = Branch(execute=@n11.else, Condition=Get l_hasFoundOneCompErrror() -> (l_hasFoundOneCompErrror:bool))
   n13 = InterlAddError(execute=[@n12.then, @n15.then], ErrorIn=Get l_error(), NewError=@n10.Error)
   > Set l_error(l_error=@n13)
   Set l_hasFoundOneCompErrror(execute=@n12.else, l_hasFoundOneCompErrror=true)
   > n14 = InterlAddError(ErrorIn=Get l_error(), NewError=------- Possible Component Issues -------)
   > n15 = Set l_error(l_error=@n14)
   n16 = Get l_error()
   Return Node!(execute=@n2.then_3, Success?:bool=IsEmpty(InString=@n16), Error:String=@n16, IsError:bool=Get l_isError() -> (l_isError:bool))
  ]}
 RunComponentLogic {nodes [
   n1 = RunComponentLogic!() -> (then, Comp:Static Mesh Component Object Reference) {is=title K2Node_FunctionEntry}
   > n2 = Set l_comp(l_comp:Static Mesh Component Object Reference=@n1) -> (Output_Get:Static Mesh Component Object Reference)
   n3 = Branch(
    execute=[@n20.then, @n2.then]
    Condition=Get bIsEditorOnly!(self:Actor Component Object Reference=Get l_comp() -> (l_comp:Static Mesh Component Object Reference)) -> (bIsEditorOnly:bool)
   )
   n4 = Sequence!(execute=@n3.else) -> (then_0, then_1)
   > n5 = Sequence!(execute) -> (then_0, then_1, then_2:Exec, then_3:Exec, then_4:Exec, then_5:Exec)
   > n6 = GetEditorProperty(Object=Get l_comp(), PropertyName=Mobility) -> (PropertyValue:EComponentMobility Enum)
   > Branch(Condition="Not Equal (Enum)"!(A:EComponentMobility Enum=@n6, B:EComponentMobility Enum=Static) -> (ReturnValue:bool))
   > n7 = InterlAddError(ErrorIn=Get l_error(), NewError=Append!(A="    - ", B, C:String=" is not Static Mobility") -> (ReturnValue))
   > Set l_error(l_error=@n7)
   n8 = Get l_comp()
   n9 = Branch(
    execute=@n5.then_1
    Condition=NOR Boolean!(
     A:bool=Is Excluded From HLOD Level!(self:Primitive Component Object Reference=@n8, HLODLevel:EHLODLevelExclusion Enum=HLOD0) -> (ReturnValue:bool)
     B:bool=Is Excluded From HLOD Level(self=@n8, HLODLevel=HLOD1)
    ) -> (ReturnValue:bool)
   )
   n10 = InterlAddError(execute=@n9.else, ErrorIn=Get l_error(), NewError=Append(A="    - ", C=" has ExcludeFromHLODLevels set."))
   > Set l_error(l_error=@n10)
   n11 = Branch(
    execute=@n5.then_2
    Condition=Get bEnableAutoLODGeneration!(self:Primitive Component Object Reference=Get l_comp()) -> (bEnableAutoLODGeneration:bool)
   )
   n12 = InterlAddError(execute=@n11.else, ErrorIn=Get l_error(), NewError=Append!(A="    - ", B=" has IncludeComponentInHLOD=False") -> (ReturnValue))
   > Set l_error(l_error=@n12)
   Branch(execute=@n5.then_3, Condition=Get bHiddenInGame!(self:Scene Component Object Reference=Get l_comp()) -> (bHiddenInGame:bool))
   > n13 = InterlAddError(ErrorIn=Get l_error(), NewError=Append(A="    - ", B="is HiddenInGame=True and will not appear in the HLOD."))
   > Set l_error(l_error=@n13)
   n14 = Branch(execute=@n5.then_4, Condition=Get bVisible!(self:Scene Component Object Reference=Get l_comp()) -> (bVisible:bool))
   n15 = InterlAddError(execute=@n14.else, ErrorIn=Get l_error(), NewError=Append(A="    - ", B="is Visible=False and will not appear in the HLOD."))
   > Set l_error(l_error=@n15)
   n16 = Is Valid(exec=@n5.then_5, InputObject=Get StaticMesh(self=Get l_comp()))
   n17 = InterlAddError(
    execute=@n16.Is Not Valid
    ErrorIn=Get l_error()
    NewError=Append(A="    - ", B=Has no mesh assigned so will not include self or any children)
   )
   > Set l_error(l_error=@n17)
   n18 = Get l_error()
   n19 = Get l_comp()
   Return Node!(
    execute=@n4.then_1
    Success?=IsEmpty(InString=@n18)
    Error=Append!(
     A="🔸"
     B=GetCompName!(Object:Object Reference=@n19) -> (ReturnValue:String) {is=title K2Node_CallFunction}
     C:String=" Component Issues\r\n"
     D:String=@n18
    ) -> (ReturnValue)
   )
   n20 = Branch(Condition=IsVisible(self:Scene Component Object Reference=Get l_comp()))
  ]}
 RunFilterLogicOnObject {nodes [
   n1 = RunFilterLogicOnObject!() -> (then, object:Object Reference) {is=title K2Node_FunctionEntry}
   > n2 = Sequence!(execute) -> (then_0, then_1)
   > n3 = Cast To Actor!(execute, Object=@n1) -> (then, CastFailed, AsActor:Actor Object Reference) {is=title K2Node_DynamicCast}
   > n4 = RunActorLogic!(execute, self, Actor:Actor Object Reference=@n3) -> (then, Success?:bool, Error:String, IsError:bool) {is=title K2Node_CallFunction}
   > Set l_success(l_success:bool=@n4) -> (Output_Get:bool)
   > Set l_error(l_error=@n4.Error)
   > Set l_isError(l_isError=@n4.IsError)
   n5 = Cast To StaticMeshComponent(execute=@n3.CastFailed, Object=@n1)
   > n6 = RunComponentLogic(Comp=@n5) {is=title K2Node_CallFunction}
   > Set l_success(l_success=@n6)
   > Set l_error(l_error=@n6.Error)
   > Set l_isError(l_isError=true)
   n7 = Get l_isError()
   n8 = Get l_isError()
   Return Node!(
    execute=@n2.then_1
    Valid:bool=Get l_success() -> (l_success:bool)
    ReportData:Struct GART Report Data Structure=Make Struct GART Report Data(
     DisplayTitle_20_#1=HLOD
     DisplayDescShort_29_#2=Select(Option 0:String=Possible Issues, Option 1:String=Errors, Index=@n8) -> (ReturnValue:String)
     DisplayDescLong_30_#3=Get l_error()
     DisplayColorOverride_40_#4=Select(Option 0:Linear Color Structure=(R=0.599,G=0.0893,B=0.0374,A=1), Option 1:Linear Color Structure=(R=0.3542,G=0.0258,B=0.0258,A=1), Index=@n7) -> (ReturnValue:Linear Color Structure)
    )
   )
  ]}
]
variables []
