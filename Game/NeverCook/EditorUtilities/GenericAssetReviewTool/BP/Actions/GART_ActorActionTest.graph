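# Editor utility of the Generic Asset Review Tool with actor actions for reviewing and isolating level actors by HLOD inclusion.
# CheckIfHLODSetup runs the selected objects through the Report_ActorInHLOD_Base filter and shows results in an EUW_GART_List_Main tab.
# HideAssetsInHLOD and HideAssetsNotInHLOD share HideActorsBasedOnHLOD, which excludes level instances, splits remaining level actors by HLOD inclusion and temporarily hides the chosen group.
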EditorUtilityBlueprint /Game/NeverCook/EditorUtilities/GenericAssetReviewTool/BP/Actions/GART_ActorActionTest
graphs [
 CheckIfHLODSetup {nodes [
   CheckIfHLODSetup() {is=title K2Node_FunctionEntry}
   > n1 = GetSelectionSet(self=Get LevelEditorSubsystem())
   > n2 = GetSelectedObjects!(execute, self:Typed Element Selection Set Object Reference=@n1, InRequiredClass:Object Class Reference) -> (then, ReturnValue:Array of Object References)
   > n3 = RunObjectsThroughFilters!(
    execute
    Objects:Array of Object References=@n2
    ReportClasses:Array of GART Report Base Soft Class References=Get Report Classes() -> (Report Classes:Array of GART Report Base Soft Class References)
    __WorldContext:Object Reference=@n1
   ) -> (then, ReportDataMap:Map of Object References to Struct GART Asset Report List Structures)
   > n4 = SpawnAndRegisterTabAndGetID!(
    execute
    self:Editor Utility Subsystem Object Reference=Get EditorUtilitySubsystem()
    InBlueprint:Editor Utility Widget Blueprint Object Reference=/Game/NeverCook/EditorUtilities/GenericAssetReviewTool/Widgets/EUW_GART_List_Main
   ) -> (then, NewTabID:Name=None, ReturnValue:Editor Utility Widget Object Reference)
   > n5 = Cast To EUW_GART_List_Main!(execute, Object=@n4.ReturnValue) -> (then, CastFailed, AsEUW GART List Main:EUW GART List Main Object Reference) {is=title K2Node_DynamicCast}
   > SetObjectReportData!(
    execute
    self:EUW GART List Main Object Reference=@n5
    ReportData:Map of Object References to Struct GART Asset Report List Structures=@n3
    UpdateList:bool=true
   ) -> (then)
  ]}
 EventGraph {links [], nodes []}
 HideActorsBasedOnHLOD {nodes [
   HideActorsBasedOnHLOD!() -> (then, HideEnabled:bool=~) {is=title K2Node_FunctionEntry}
   > n1 = GetAllLevelActors(self=Get EditorActorSubsystem())
   > n2 = Filter by Class!(
    execute
    TargetArray:Array of Object References=@n1
    ObjectClass:Object Class Reference=/Script/FortniteGame.BuildingLevelInstance
    FilterType:EEditorScriptingFilterType Enum=Exclude
   ) -> (then, ReturnValue:Array of Building Level Instance Object References)
   > n3 = Filter by Class(TargetArray=@n2, ObjectClass=/Script/Engine.LevelInstance, FilterType=Exclude) -> (ReturnValue:Array of Level Instance Object References)
   > n4 = Set l_actors(l_actors:Array of Actor Object References=@n3) -> (Output_Get:Array of Actor Object References)
   > n5 = ActorsAreIncludedInHLOD!(execute, Actors:Array of Actor Object References=@n4, __WorldContext:Object Reference=Get l_world()) -> (then, Excluded:Array of Actor Object References, Included:Array of Actor Object References)
   > n6 = Branch(Condition=Get HideEnabled() -> (HideEnabled:bool))
   > n7 = For Each Loop(Array:Array of Actor Object References=@n5.Included) -> (Array Element:Actor Object Reference)
   > SetIsTemporarilyHiddenInEditor!(execute, self:Actor Object Reference=@n7, bIsHidden:bool=true) -> (then)
   n8 = For Each Loop(Exec=@n6.else, Array=@n5)
   > SetIsTemporarilyHiddenInEditor(self=@n8, bIsHidden=true)
  ]}
 HideAssetsInHLOD {nodes [
   HideAssetsInHLOD() {is=title K2Node_FunctionEntry}
   > HideActorsBasedOnHLOD!(execute, self, HideEnabled:bool=true) -> (then) {is=title K2Node_CallFunction}
  ]}
 HideAssetsNotInHLOD {nodes [
   HideAssetsNotInHLOD() {is=title K2Node_FunctionEntry}
   > HideActorsBasedOnHLOD(HideEnabled=false)
  ]}
]
variables [Report Classes {
  type="array<Class>"
  default [/Game/NeverCook/EditorUtilities/GenericAssetReviewTool/BP/ReportExamples/Actor/Report_ActorInHLOD_Base.Report_ActorInHLOD_Base_C]
 }]
