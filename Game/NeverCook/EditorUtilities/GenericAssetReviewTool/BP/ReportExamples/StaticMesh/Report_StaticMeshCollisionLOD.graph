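# Editor report for the Generic Asset Review Tool that flags static meshes whose LOD For Collision is unsuitable for bullet collision.
# When loading is allowed, LoadLogic exits early for simple-as-complex collision, otherwise reads per-LOD vertex counts.
# It flags a missing LOD, LOD0 over a minimum, or counts over error/warning thresholds, raising severity that determines validity.
# The FX-only check ends in branches with no effect.
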EditorUtilityBlueprint /Game/NeverCook/EditorUtilities/GenericAssetReviewTool/BP/ReportExamples/StaticMesh/Report_StaticMeshCollisionLOD
graphs [
 EventGraph {links [], nodes []}
 LoadLogic {nodes [
   n1 = LoadLogic!() -> (then, AssetData:Asset Data Structure) {is=title K2Node_FunctionEntry}
   > Set l_asset(l_asset:Asset Data Structure=@n1) -> (Output_Get:Asset Data Structure)
   > n2 = Cast To StaticMesh(Object=GetAsset(InAssetData=@n1))
   > Set l_mesh(l_mesh:Static Mesh Object Reference=@n2) -> (Output_Get:Static Mesh Object Reference)
   > Set MaxVertsForCollisionLOD_Error(MaxVertsForCollisionLOD_Error:int="Max (Integer)"(
     A=Get MaxVertsForCollisionLOD_Warning() -> (MaxVertsForCollisionLOD_Warning:int=0)
     B=Get MaxVertsForCollisionLOD_Error() -> (MaxVertsForCollisionLOD_Error:int=0)
    )) -> (Output_Get:int=0)
   > n3 = GetEditorWorld(self=Get UnrealEditorSubsystem())
   > Set l_world(l_world:Object Reference=@n3) -> (Output_Get:Object Reference)
   > n4 = Sequence!(execute) -> (then_0, then_1, then_2:Exec, then_3:Exec)
   > n5 = NoLoad_GetStaticMeshCollisionData!(execute, AssetData:Asset Data Structure=Get l_asset() -> (l_asset:Asset Data Structure), __WorldContext:Object Reference=Get l_world()) -> (then, SectionsWithCollision:int=0, CollisionPrimCount:int=0, CollisionComplexity:String)
   > n6 = Branch(Condition="Not Equal Exactly (String)"(A=@n5.CollisionComplexity, B=CTF_UseSimpleAsComplex))
   > n7 = LoadAssetAndGetVertexCounts!(execute, AssetData:"Asset Data Structure (by ref)"=Get l_asset(), __WorldContext:Object Reference=Get l_world()) -> (then, VertsPerLOD:Array of Integers)
   > Set l_vertsPerLOD(l_vertsPerLOD:Array of Integers=@n7) -> (Output_Get:Array of Integers)
   Return Node!(execute=@n6.else, Valid:bool, ReportData:Struct GART Report Data Structure)
   n8 = IsMeshOnlyUsedInFX!(execute=@n4.then_1, assetData:Asset Data Structure=Get l_asset(), __WorldContext:Object Reference=Get l_world()) -> (then, OnlyFX:bool)
   > Branch(Condition=@n8)
   > n9 = NoLoad_GetStaticMeshCollisionData(AssetData=Get l_asset(), __WorldContext=Get l_world()) -> (SectionsWithCollision=0, CollisionPrimCount=0)
   > Branch(Condition="integer > integer"(A=@n9))
   n10 = "Max (Integer)"(
    A=Get LODForCollision!(self:Static Mesh Object Reference=Get l_mesh() -> (l_mesh:Static Mesh Object Reference)) -> (LODForCollision:int=0)
    B=0
   )
   Set l_vertCount(execute=@n4.then_2, l_vertCount:int="Get (a ref)"(Array=Get l_vertsPerLOD() -> (l_vertsPerLOD:Array of Integers), Dimension 1=@n10)) -> (Output_Get:int=0)
   > n11 = Branch(Condition=Is Valid Index!(TargetArray:Array of Integers=Get l_vertsPerLOD(), IndexToTest:int=@n10) -> (ReturnValue:bool))
   n12 = Get LODForCollision(self=Get l_mesh()) -> (LODForCollision=0)
   n13 = Get l_vertCount() -> (l_vertCount:int=0)
   n14 = Get VertCountToAlwaysRequireLOD() -> (VertCountToAlwaysRequireLOD:int=0)
   > n15 = Branch(Condition=AND Boolean(A="integer < integer"(A=@n12, B=1), B="integer > integer"(A=@n13, B=@n14)))
   n16 = Get LODForCollision(self=Get l_mesh()) -> (LODForCollision=0)
   n17 = Get l_vertCount() -> (l_vertCount=0)
   > AddError(NewError=Append!(
     A="LOD for Collision is set to "
     B="To String (Integer)"(InInt=@n16)
     C:String=" and LOD0 is "
     D:String="To String (Integer)"(InInt=@n17)
     E:String=" verts.  Try to set the LOD For Collision to the LOD that is most efficient but accurate enough for bullet collision."
    ) -> (ReturnValue))
   > IncreaseErrorSeverity!(
    execute
    self:GART Report Base Object Reference
    Severity:"Enum_GART_ReportSeverity Enum (by ref)"=Get l_severity()
    NewSeverity:Enum_GART_ReportSeverity Enum=NewEnumerator2
   ) -> (then)
   n18 = Branch(
    execute=@n15.else
    Condition="integer > integer"(A=Get l_vertCount() -> (l_vertCount=0), B=Get MaxVertsForCollisionLOD_Error() -> (MaxVertsForCollisionLOD_Error=0))
   )
   n19 = Get l_vertCount() -> (l_vertCount=0)
   n20 = Get MaxVertsForCollisionLOD_Error() -> (MaxVertsForCollisionLOD_Error=0)
   n21 = Get LODForCollision(self=Get l_mesh()) -> (LODForCollision=0)
   > AddError(NewError=Append!(
     A="LOD For Collision is set to an LOD with too many vertices.  LOD "
     B="To String (Integer)"(InInt=@n21)
     C=" has "
     D="To String (Integer)"(InInt=@n19)
     E=" and must be below "
     F:String="To String (Integer)"(InInt=@n20)
    ) -> (ReturnValue))
   > IncreaseErrorSeverity(Severity=Get l_severity(), NewSeverity=NewEnumerator2)
   Branch(
    execute=@n18.else
    Condition="integer > integer"(A=Get l_vertCount() -> (l_vertCount=0), B=Get MaxVertsForCollisionLOD_Warning() -> (MaxVertsForCollisionLOD_Warning=0))
   )
   n22 = Get LODForCollision(self=Get l_mesh()) -> (LODForCollision=0)
   n23 = Get l_vertCount() -> (l_vertCount=0)
   n24 = Get MaxVertsForCollisionLOD_Warning() -> (MaxVertsForCollisionLOD_Warning=0)
   > AddError(NewError=Append(
     A="LOD For Collision is set to an LOD with a lot of vertices.  LOD "
     B="To String (Integer)"(InInt=@n22)
     C=" has "
     D="To String (Integer)"(InInt=@n23)
     E=" verts and most meshes don't need more than "
     F="To String (Integer)"(InInt=@n24)
    ))
   > IncreaseErrorSeverity(Severity=Get l_severity(), NewSeverity=NewEnumerator1)
   n25 = Get LODForCollision(self=Get l_mesh()) -> (LODForCollision=0)
   n26 = Length(TargetArray:Array of Integers=Get l_vertsPerLOD())
   AddError(
    execute=@n11.else
    NewError=Append!(
     A="LOD For Collision is set to an LOD that doesn't exist.  It's set to "
     B="To String (Integer)"(InInt=@n25)
     C=" but there are only "
     D="To String (Integer)"(InInt=@n26)
     E=" LODs on the mesh."
    ) -> (ReturnValue)
   )
   > IncreaseErrorSeverity(Severity=Get l_severity(), NewSeverity=NewEnumerator2)
   Return Node(
    execute=@n4.then_3
    Valid="Equal (Enum)"(A:Enum_GART_ReportSeverity Enum=Get l_severity(), B:Enum_GART_ReportSeverity Enum=NewEnumerator0)
    ReportData=Make Struct GART Report Data(
     DisplayTitle_20_#1=Collision
     DisplayDescShort_29_#2=LOD For Collision
     Severity_46_#3=Get l_severity()
    )
   )
  ]}
 RunFilterLogicOnAssetData {nodes [
   n1 = RunFilterLogicOnAssetData()
   > n2 = Branch(Condition=@n1.AllowLoading)
   > n3 = LoadLogic!(execute, self, AssetData:Asset Data Structure=@n1) -> (then, Valid:bool, ReportData:Struct GART Report Data Structure) {is=title K2Node_CallFunction}
   > Return Node(Valid=@n3, ReportData=@n3.ReportData)
   Return Node(execute=@n2.else)
  ]}
]
variables [
 MaxVertsForCollisionLOD_Error {type=integer, default=5000}
 MaxVertsForCollisionLOD_Warning {type=integer, default=1500}
 VertCountToAlwaysRequireLOD {type=integer, default=150}
]
