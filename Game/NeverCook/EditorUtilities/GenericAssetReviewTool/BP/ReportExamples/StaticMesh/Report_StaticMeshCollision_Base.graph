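# Editor utility base report for the Generic Asset Review Tool that flags static meshes with improper collision settings.
# RunFilterLogicOnAssetData calls NoLoadLogic, which reads section count, primitive count and complexity from asset data via NoLoad_GetStaticMeshCollisionData.
# Depending on complexity mode, it checks complex-collision sections and simple primitive counts against variable limits.
# Failures build error text, description and severity; the result is valid only at the default severity.
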EditorUtilityBlueprint /Game/NeverCook/EditorUtilities/GenericAssetReviewTool/BP/ReportExamples/StaticMesh/Report_StaticMeshCollision_Base
graphs [
 EventGraph {links [], nodes []}
 NoLoadLogic {nodes [
   n1 = AppendErrorRow!(execute, self:GART Report Base Object Reference, ErrorIn:"String (by ref)"=Get l_error() -> (l_error:String), NewError:String) -> (then)
   n2 = Sequence(execute=[@n1.then, @n24.then])
   > n3 = Branch(Condition="Equal Exactly (String)"(A=Get l_descShort() -> (l_descShort:String)))
   > Set l_descShort(l_descShort:String=Bad Simple Collision) -> (Output_Get:String)
   Set l_descShort(execute=@n3.else, l_descShort=Bad Complex And Simple Collision)
   n4 = Get MaxSimplePrimsAllowedBeforeError() -> (MaxSimplePrimsAllowedBeforeError:int=0)
   n5 = Get l_primCount() -> (l_primCount:int=0)
   IncreaseErrorSeverity!(
    execute=@n2.then_1
    self:GART Report Base Object Reference
    Severity:"Enum_GART_ReportSeverity Enum (by ref)"=Get l_severity()
    NewSeverity:Enum_GART_ReportSeverity Enum=Select(
     Option 0:Enum_GART_ReportSeverity Enum=NewEnumerator1
     Option 1:Enum_GART_ReportSeverity Enum=NewEnumerator2
     Index="integer > integer"(A=@n5, B=@n4)
    ) -> (ReturnValue:Enum_GART_ReportSeverity Enum)
   ) -> (then)
   n6 = NoLoadLogic!() -> (then, AssetData:Asset Data Structure) {is=title K2Node_FunctionEntry}
   > n7 = NoLoad_GetStaticMeshCollisionData!(execute, AssetData:Asset Data Structure=@n6, __WorldContext:Object Reference=Get l_world()) -> (then, SectionsWithCollision:int=0, CollisionPrimCount:int=0, CollisionComplexity:String)
   > Set l_sections(l_sections:int=@n7) -> (Output_Get:int=0)
   > Set l_primCount(l_primCount:int=@n7.CollisionPrimCount) -> (Output_Get:int=0)
   > Set l_complexity(l_complexity:String=@n7.CollisionComplexity) -> (Output_Get:String)
   > Set l_error(l_error:String=Improper Collision Settings) -> (Output_Get:String)
   > n8 = Sequence!(execute) -> (then_0, then_1, then_2:Exec)
   > Branch(Condition="Not Equal Exactly (String)"(A=Get l_complexity() -> (l_complexity:String), B=CTF_UseSimpleAsComplex))
   > Branch(Condition="integer > integer"(A=Get l_sections() -> (l_sections:int=0), B=Get MaxSectionsWithCollisionAllowed() -> (MaxSectionsWithCollisionAllowed:int=0)))
   n9 = Get l_sections() -> (l_sections=0)
   n10 = Get MaxSectionsWithCollisionAllowed() -> (MaxSectionsWithCollisionAllowed=0)
   > Set l_error(l_error=Append!(
     A=Get l_error()
     B="\r\n- Has "
     C:String="To String (Integer)"(InInt=@n9)
     D:String=" sections with complex collision and should have "
     E:String="To String (Integer)"(InInt=@n10)
    ) -> (ReturnValue))
   > Set l_descShort(l_descShort=Bad Complex Collision)
   > Set l_severity(l_severity:Enum_GART_ReportSeverity Enum=NewEnumerator2) -> (Output_Get:Enum_GART_ReportSeverity Enum)
   Branch(execute=@n8.then_1, Condition="Not Equal Exactly (String)"(A=Get l_complexity(), B=CTF_UseComplexAsSimple))
   n11 = "Max (Integer)"(
    A=Get MaxSimplePrimsAllowedBeforeWarning() -> (MaxSimplePrimsAllowedBeforeWarning:int=0)
    B=Get MaxSimplePrimsAllowedBeforeError() -> (MaxSimplePrimsAllowedBeforeError=0)
   )
   n12 = Get l_primCount() -> (l_primCount=0)
   n13 = Get l_primCount() -> (l_primCount=0)
   n14 = Get MinSimplePrimsAllowed() -> (MinSimplePrimsAllowed:int=0)
   > Branch(Condition=OR Boolean(A="integer > integer"(A=@n12, B=@n11), B="integer < integer"(A=@n13, B=@n14)))
   n15 = "To String (Integer)"(InInt=int + int(A=Get MaxSimplePrimsAllowedBeforeWarning() -> (MaxSimplePrimsAllowedBeforeWarning=0)))
   n16 = "To String (Integer)"(InInt=Get l_primCount() -> (l_primCount=0))
   n17 = int - int(A=Get MinSimplePrimsAllowed() -> (MinSimplePrimsAllowed=0))
   n18 = "To String (Integer)"(InInt=@n17)
   n19 = "To String (Integer)"(InInt=Get l_primCount() -> (l_primCount=0))
   n20 = "To String (Integer)"(InInt=@n17)
   n21 = "To String (Integer)"(InInt=int + int(A=Get MaxSimplePrimsAllowedBeforeError() -> (MaxSimplePrimsAllowedBeforeError=0)))
   n22 = Get l_primCount() -> (l_primCount=0)
   n23 = Get MaxSimplePrimsAllowedBeforeError() -> (MaxSimplePrimsAllowedBeforeError=0)
   > n24 = AddError(NewError=SelectString(
     A=Append!(
      A="Has  "
      B=@n19
      C=" simple collision primitives and should have more than "
      D=@n20
      E=" but must be less than "
      F:String=@n21
      G:String=" unless it's extremely important."
     ) -> (ReturnValue)
     B=Append!(A="Has  ", B=@n16, C=" simple collision primitives and should have more than ", D=@n18, E=" and try to have less than ", F=@n15) -> (ReturnValue)
     bPickA="integer > integer"(A=@n22, B=@n23)
    ))
   Return Node!(
    execute=@n8.then_2
    Valid:bool="Equal (Enum)"(A:Enum_GART_ReportSeverity Enum=Get l_severity(), B:Enum_GART_ReportSeverity Enum=NewEnumerator0)
    ReportData:Struct GART Report Data Structure=Make Struct GART Report Data(
     DisplayTitle_20_#1=Collision
     DisplayDescShort_29_#2=Get l_descShort()
     Severity_46_#3=Get l_severity()
    )
   )
  ]}
 RunFilterLogicOnAssetData {nodes [
   n1 = RunFilterLogicOnAssetData()
   > n2 = NoLoadLogic!(execute, self, AssetData:Asset Data Structure=@n1) -> (then, Valid:bool, ReportData:Struct GART Report Data Structure) {is=title K2Node_CallFunction}
   > Return Node(Valid=@n2, ReportData=@n2.ReportData)
  ]}
]
variables [
 MaxSectionsWithCollisionAllowed {type=integer, default=999}
 MaxSimplePrimsAllowedBeforeError {type=integer, default=5}
 MaxSimplePrimsAllowedBeforeWarning {type=integer, default=2}
 MaxVertCountForComplexCollision {type=integer, default=2500}
 MinSectionsWithCollisionAllowed {type=integer, default=1}
 MinSimplePrimsAllowed {type=integer, default=1}
]
