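# Editor utility report for the Generic Asset Review Tool that checks Static Mesh Nanite settings.
# CheckNanite reads the NaniteEnabled tag and, when loading is allowed, runs DoLoadTests to inspect material blend modes and fallback settings against configurable targets and ranges.
# Meshes with a HasHiResMesh tag skip fallback checks.
# Disabled Nanite is flagged unless a translucent or additive material is found, which resets final severity to the base level.
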EditorUtilityBlueprint /Game/NeverCook/EditorUtilities/GenericAssetReviewTool/BP/ReportExamples/StaticMesh/Report_StaticMeshNanite_Base
graphs [
 CheckNanite {nodes [
   n1 = CheckNanite!() -> (then, AssetData:Asset Data Structure, allowLoading:bool=~) {is=title K2Node_FunctionEntry}
   > Set l_assetData(l_assetData=@n1)
   > Set l_allowLoading(l_allowLoading:bool=@n1.allowLoading) -> (Output_Get:bool)
   > n2 = Sequence!(execute) -> (then_0, then_1, then_2:Exec, then_3:Exec, then_4:Exec, then_5:Exec)
   n3 = GetTagValue(InAssetData=Get l_assetData(), InTagName=MakeLiteralName(Value=NaniteEnabled))
   > Branch(Condition=@n3.ReturnValue)
   > Set l_naniteEnabled(l_naniteEnabled:bool="Equal Exactly (String)"(A=@n3, B=True)) -> (Output_Get:bool)
   Branch(execute=@n2.then_1, Condition=Get l_allowLoading() -> (l_allowLoading:bool))
   > n4 = DoLoadTests!(execute, self, InAssetData:"Asset Data Structure (by ref)"=Get l_assetData()) -> (then, Errors:String, Severity:Enum_GART_ReportSeverity Enum, isValid:bool, atLeast1BadMat:bool)
   > Set l_loadErrors(l_loadErrors:String=@n4) -> (Output_Get:String)
   > IncreaseErrorSeverity!(
    execute
    self:GART Report Base Object Reference
    Severity:"Enum_GART_ReportSeverity Enum (by ref)"=Get l_finalSeverity() -> (l_finalSeverity:Enum_GART_ReportSeverity Enum)
    NewSeverity:Enum_GART_ReportSeverity Enum=@n4.Severity
   ) -> (then)
   > Set l_valid(l_valid:bool=@n4.isValid) -> (Output_Get:bool)
   > Set l_atLeast1BadMat(l_atLeast1BadMat:bool=@n4.atLeast1BadMat) -> (Output_Get:bool)
   n5 = Branch(execute=@n2.then_2, Condition=Get l_naniteEnabled() -> (l_naniteEnabled:bool))
   n6 = Branch(execute=@n5.else, Condition=Get l_atLeast1BadMat() -> (l_atLeast1BadMat:bool))
   AddError(execute=@n6.else, NewError=Nanite is not enabled.)
   > IncreaseErrorSeverity(Severity=Get l_finalSeverity(), NewSeverity=NewEnumerator2)
   n7 = Get l_naniteEnabled()
   Branch(execute=@n2.then_4, Condition=AND Boolean(A=NOT Boolean(A=@n7), B=Get l_atLeast1BadMat()))
   > Set l_finalSeverity(l_finalSeverity:Enum_GART_ReportSeverity Enum=NewEnumerator0) -> (Output_Get:Enum_GART_ReportSeverity Enum)
   Return Node!(
    execute=@n2.then_5
    Valid:bool="Equal (Enum)"(A:Enum_GART_ReportSeverity Enum=Get l_finalSeverity(), B:Enum_GART_ReportSeverity Enum=NewEnumerator0)
    ReportData:Struct GART Report Data Structure=Make Struct GART Report Data(
     DisplayTitle_20_#1=Nanite
     DisplayDescShort_29_#2=Nanite Settings
     DisplayDescLong_30_#3=Get l_finalLongDesc() -> (l_finalLongDesc:String)
     Severity_46_#4=Get l_finalSeverity()
    )
   )
  ]}
 DoLoadTests {nodes [
   n1 = DoLoadTests!() -> (then, InAssetData:"Asset Data Structure (by ref)") {is=title K2Node_FunctionEntry}
   > Set l_assetData(l_assetData=@n1)
   > n2 = Cast To StaticMesh(Object=GetAsset(InAssetData=@n1))
   > Set l_mesh(l_mesh:Static Mesh Object Reference=@n2) -> (Output_Get:Static Mesh Object Reference)
   > n3 = Sequence!(execute) -> (then_0, then_1, then_2, then_3, then_4)
   n4 = Break Mesh Nanite Settings!(MeshNaniteSettings:"Mesh Nanite Settings Structure (by ref)"=GetNaniteSettings!(
     self:Static Mesh Editor Subsystem Object Reference=Get StaticMeshEditorSubsystem()
     StaticMesh:Static Mesh Object Reference=Get l_mesh() -> (l_mesh:Static Mesh Object Reference)
    ) -> (ReturnValue:Mesh Nanite Settings Structure)) -> (
    bEnabled:bool=~
    bExplicitTangents:bool=~
    bLerpUVs:bool=~
    bSeparable:bool=~
    bVoxelNDF:bool=~
    bVoxelOpacity:bool=~
    ShapePreservation:ENaniteShapePreservation Enum
    PositionPrecision:int
    NormalPrecision:int
    TangentPrecision:int
    BoneWeightPrecision:int
    KeepPercentTriangles:float=~
    TrimRelativeError:float=~
    GenerateFallback:ENaniteGenerateFallback Enum
    FallbackTarget:ENaniteFallbackTarget Enum
    FallbackPercentTriangles:float=~
    FallbackRelativeError:float=~
    MaxEdgeLengthFactor:float=~
    NumRays:int
    VoxelLevel:int
    RayBackUp:float=~
    DisplacementUVChannel:int
   ) {is=title K2Node_BreakStruct}
   > Set l_keepTriangles(l_keepTriangles:float=float * float(A=@n4.FallbackPercentTriangles, B=100) {is=title K2Node_PromotableOperator}) -> (Output_Get:float)
   > Set l_fallbackTarget(l_fallbackTarget:ENaniteFallbackTarget Enum=@n4.FallbackTarget) -> (Output_Get:ENaniteFallbackTarget Enum=Auto)
   > Set l_relativeError(l_relativeError:float=@n4.FallbackRelativeError) -> (Output_Get:float)
   > n5 = Set l_naniteEnabled(l_naniteEnabled=@n4)
   > Branch(Condition=@n5)
   n6 = MakeLiteralName(Value=HasHiResMesh)
   n7 = Get l_assetData()
   > Set l_hasHiRes(l_hasHiRes:bool="Equal Exactly (String)"(A=GetTagValue(InAssetData=@n7, InTagName=@n6), B=True)) -> (Output_Get:bool)
   n8 = For Each Loop with Break!(
    Exec:Exec=@n3.then_1
    Array:Array of Static Material Structures=Get StaticMaterials!(self:Static Mesh Object Reference=Get l_mesh()) -> (StaticMaterials:Array of Static Material Structures)
    Break:Exec=@n10.then
   ) -> (LoopBody:Exec, Array Element:Static Material Structure, Array Index:int, Completed:Exec)
   Branch(execute=@n8.Completed, Condition=AND Boolean(A=Get l_naniteEnabled(), B=Get l_hasAtLeast1BadMat() -> (l_hasAtLeast1BadMat:bool)))
   > AddError(NewError=Has at least 1 material with translucent or additive blend mode which is not supported on Nanite.)
   > IncreaseErrorSeverity(Severity=Get l_severity(), NewSeverity=NewEnumerator2)
   n9 = GetBlendMode!(self:Material Interface Object Reference=Break Static Material!(StaticMaterial:"Static Material Structure (by ref)"=@n8) -> (
     MaterialInterface:Material Interface Object Reference
     MaterialSlotName:Name
     UVChannelData:Mesh UVChannel Info Structure
     OverlayMaterialInterface:Material Interface Object Reference
    ) {is=title K2Node_BreakStruct}) -> (ReturnValue:EBlendMode Enum=BLEND_Opaque)
   Branch(
    execute=@n8.LoopBody
    Condition=OR Boolean(A="Equal (Enum)"(A:EBlendMode Enum=@n9, B:EBlendMode Enum=BLEND_Translucent), B="Equal (Enum)"(A=@n9, B=BLEND_Additive))
   )
   > n10 = Set l_hasAtLeast1BadMat(l_hasAtLeast1BadMat:bool=true) -> (Output_Get:bool)
   n11 = Branch(execute=@n3.then_2, Condition=Get l_hasHiRes() -> (l_hasHiRes:bool))
   Branch(
    execute=@n11.else
    Condition="Not Equal (Enum)"!(
     A:ENaniteFallbackTarget Enum=Get l_fallbackTarget() -> (l_fallbackTarget:ENaniteFallbackTarget Enum=Auto)
     B:ENaniteFallbackTarget Enum=Get FallbackTarget() -> (FallbackTarget:ENaniteFallbackTarget Enum=Auto)
    ) -> (ReturnValue:bool)
   )
   n12 = Get FallbackTarget() -> (FallbackTarget=Auto)
   > AddError(NewError=Append(A="Nanite Fallback Target not set to ", B=Enum to String!(Enumerator:ENaniteFallbackTarget Enum=@n12) -> (ReturnValue:String)))
   > IncreaseErrorSeverity(
    Severity=Get l_severity()
    NewSeverity=Select(Option 0:Enum_GART_ReportSeverity Enum=NewEnumerator1, Option 1:Enum_GART_ReportSeverity Enum=NewEnumerator2, Index=Get l_naniteEnabled()) -> (ReturnValue:Enum_GART_ReportSeverity Enum)
   )
   n13 = Branch(execute=@n3.then_3, Condition=Get l_hasHiRes())
   n14 = Branch(
    execute=@n13.else
    Condition="Equal (Enum)"(A:ENaniteFallbackTarget Enum=Get l_fallbackTarget() -> (l_fallbackTarget=Auto), B:ENaniteFallbackTarget Enum=PercentTriangles)
   )
   n15 = BreakVector2D(InVec=Get KeepTrianglesRange() -> (KeepTrianglesRange:Vector 2D Structure))
   n16 = "Min (Float)"(A:double=@n15, B:double=@n15.Y) -> (ReturnValue:double) {is=title K2Node_CommutativeAssociativeBinaryOperator}
   n17 = "Max (Float)"(A=@n15, B=@n15.Y)
   > n18 = Branch(Condition="In Range (Float)"!(
     Value:double=Get l_keepTriangles() -> (l_keepTriangles:float)
     Min:double=@n16
     Max:double=@n17
     InclusiveMin:bool=true
     InclusiveMax:bool=true
    ) -> (ReturnValue:bool))
   n19 = Get l_keepTriangles()
   AddError(
    execute=@n18.else
    NewError=Append!(
     A="Nanite Keep Triangle Percentage is out of range.  It is "
     B="To String (Float)"(InDouble=@n19)
     C:String=" and should be between "
     D:String="To String (Float)"(InDouble=@n16)
     E:String=" and "
     F:String="To String (Float)"(InDouble=@n17)
     G:String=.
    ) -> (ReturnValue)
   )
   > IncreaseErrorSeverity(Severity=Get l_severity(), NewSeverity=Select(Option 0=NewEnumerator1, Option 1=NewEnumerator2, Index=Get l_naniteEnabled()))
   Branch(execute=@n14.else, Condition="Equal (Enum)"(A=Get l_fallbackTarget() -> (l_fallbackTarget=Auto), B=RelativeError))
   n20 = BreakVector2D(InVec=Get FallbackRelativeErrorRange() -> (FallbackRelativeErrorRange:Vector 2D Structure))
   n21 = "Min (Float)"(A=@n20, B=@n20.Y)
   n22 = "Max (Float)"(A=@n20, B=@n20.Y)
   > n23 = Branch(Condition="In Range (Float)"(Value=Get l_relativeError() -> (l_relativeError:float), Min=@n21, Max=@n22, InclusiveMin=true, InclusiveMax=true))
   n24 = Get l_relativeError()
   AddError(
    execute=@n23.else
    NewError=Append(
     A="Nanite Fallback Relative Error is out of range.  It is "
     B="To String (Float)"(InDouble=@n24)
     C=" and should be between "
     D="To String (Float)"(InDouble=@n21)
     E=" and "
     F="To String (Float)"(InDouble=@n22)
     G=.
    )
   )
   > IncreaseErrorSeverity(Severity=Get l_severity(), NewSeverity=Select(Option 0=NewEnumerator1, Option 1=NewEnumerator2, Index=Get l_naniteEnabled()))
   Return Node!(
    execute=@n3.then_4
    Errors:String=Get l_error() -> (l_error:String)
    Severity:Enum_GART_ReportSeverity Enum=Get l_severity()
    isValid:bool=IsEmpty(InString=Get l_error())
    atLeast1BadMat:bool=Get l_hasAtLeast1BadMat()
   )
  ]}
 EventGraph {links [], nodes []}
 RunFilterLogicOnAssetData {nodes [
   n1 = RunFilterLogicOnAssetData()
   > n2 = CheckNanite!(execute, self, AssetData:Asset Data Structure=@n1, allowLoading:bool=@n1.AllowLoading) -> (then, Valid:bool, ReportData:Struct GART Report Data Structure) {is=title K2Node_CallFunction}
   > Return Node!(execute, Valid:bool=@n2, ReportData:Struct GART Report Data Structure=@n2.ReportData)
  ]}
]
variables [
 FallbackRelativeErrorRange {type=Vector2D, default=<0 1>}
 FallbackTarget {type=ENaniteFallbackTarget, default=PercentTriangles}
 KeepTrianglesRange {type=Vector2D, default=<100 100>}
]
