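# Editor-only asset review report base that checks static mesh vertex counts against per-quality-level limits in QualityLevelMaxVerts.
# RunFilterLogicOnAssetData branches on AllowLoading to one of two checks.
# The loaded check tests each quality's MinLOD vertex count, falling back to LOD0, flags out-of-range indices and raises severity.
# The no-load check reads asset data only, flagging an over-limit count unless that quality's MinLOD is above zero.
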EditorUtilityBlueprint /Game/NeverCook/EditorUtilities/GenericAssetReviewTool/BP/ReportExamples/StaticMesh/Report_StaticMeshVertCount_Base
graphs [
 EventGraph {links [], nodes []}
 Load_MinLODCheck {nodes [
   n1 = Load_MinLODCheck!() -> (then, AssetData:Asset Data Structure) {is=title K2Node_FunctionEntry}
   > n2 = Set l_assetData(l_assetData=@n1)
   > n3 = Cast To StaticMesh(Object=GetAsset(InAssetData=@n2))
   > Set l_staticMesh(l_staticMesh:Static Mesh Object Reference=@n3) -> (Output_Get:Static Mesh Object Reference)
   > n4 = GetEditorWorld(self=Get UnrealEditorSubsystem())
   > Set l_world(l_world:Object Reference=@n4) -> (Output_Get:Object Reference)
   > Set l_valid(l_valid:bool=true) -> (Output_Get:bool)
   > n5 = Sequence!(execute) -> (then_0, then_1, then_2:Exec)
   > AddError(NewError=Missing required MinLOD Quality Levels.)
   n6 = GetMinLODForQualityLevels!(self:Static Mesh Object Reference=Get l_staticMesh() -> (l_staticMesh:Static Mesh Object Reference)) -> (QualityLevelMinimumLODs:Map of EPerQualityLevels Enums to Integers=Low, Default:int=0)
   > n7 = Set l_meshQualityLevels(l_meshQualityLevels:Map of EPerQualityLevels Enums to Integers=@n6) -> (Output_Get:Map of EPerQualityLevels Enums to Integers=Low)
   > Add!(
    execute
    TargetMap:Map of EPerQualityLevels Enums to Integers=@n7
    Key:"EPerQualityLevels Enum (by ref)"=Num
    Value:"Integer (by ref)"=@n6.Default
   ) -> (then) {is=title K2Node_CallFunction}
   > n8 = LoadAssetAndGetVertexCounts!(execute, AssetData:"Asset Data Structure (by ref)"=Get l_assetData(), __WorldContext:Object Reference=Get l_world()) -> (then, VertsPerLOD:Array of Integers)
   > Set l_vertsPerLOD(l_vertsPerLOD:Array of Integers=@n8) -> (Output_Get:Array of Integers)
   n9 = Get QualityLevelMaxVerts() -> (QualityLevelMaxVerts:Map of EPerQualityLevels Enums to Integers=Low)
   n10 = Keys!(execute=@n5.then_1, TargetMap:Map of EPerQualityLevels Enums to Integers=@n9) -> (then, Keys:Array of EPerQualityLevels Enums)
   > n11 = Values!(execute, TargetMap:Map of EPerQualityLevels Enums to Integers=@n9) -> (then, Values:Array of Integers)
   > n12 = For Each Loop(Array:Array of EPerQualityLevels Enums=@n10) -> (Array Element:EPerQualityLevels Enum)
   > Set l_curQualityToCheck(l_curQualityToCheck:EPerQualityLevels Enum=@n12) -> (Output_Get:EPerQualityLevels Enum=Low)
   > Set l_vertCountMax(l_vertCountMax:int="Get (a ref)"(Array=@n11, Dimension 1=@n12.Array Index)) -> (Output_Get:int=0)
   n13 = Find(
    TargetMap:Map of EPerQualityLevels Enums to Integers=Get l_meshQualityLevels() -> (l_meshQualityLevels:Map of EPerQualityLevels Enums to Integers=Low)
    Key:"EPerQualityLevels Enum (by ref)"=@n12
   ) -> (Value:int)
   > n14 = Branch(Condition=@n13.ReturnValue)
   > Set l_qualityLevelIndex(l_qualityLevelIndex:int=@n13) -> (Output_Get:int=0)
   > n15 = Branch(Condition=Is Valid Index!(
     TargetArray:Array of Integers=Get l_vertsPerLOD() -> (l_vertsPerLOD:Array of Integers)
     IndexToTest:int=Get l_qualityLevelIndex() -> (l_qualityLevelIndex:int=0)
    ) -> (ReturnValue:bool))
   > Set l_vertCount(l_vertCount:int="Get (a ref)"(Array=Get l_vertsPerLOD(), Dimension 1=Get l_qualityLevelIndex() -> (l_qualityLevelIndex=0))) -> (Output_Get:int=0)
   > Branch(Condition="integer > integer"(A=Get l_vertCount() -> (l_vertCount:int=0), B=Get l_vertCountMax() -> (l_vertCountMax:int=0)))
   n16 = Get l_vertCount() -> (l_vertCount=0)
   n17 = Get l_vertCountMax() -> (l_vertCountMax=0)
   n18 = Get l_curQualityToCheck() -> (l_curQualityToCheck:EPerQualityLevels Enum=Low)
   > AddError(
    NewError=Append!(
     A="LOD Used on "
     B=Enum to String!(Enumerator:EPerQualityLevels Enum=@n18) -> (ReturnValue:String)
     C:String=" is "
     D:String="To String (Integer)"(InInt=@n16)
     E:String=" Verts and should be less than: "
     F:String="To String (Integer)"(InInt=@n17)
    ) -> (ReturnValue)
    Depth=1
   )
   > SetSeverity!(execute, self:GART Report Base Object Reference, NewSeverity:Enum_GART_ReportSeverity Enum=NewEnumerator2, OnlyIfHigher:bool=true) -> (then)
   n19 = Get l_qualityLevelIndex() -> (l_qualityLevelIndex=0)
   n20 = Length(TargetArray:Array of Integers=Get l_vertsPerLOD())
   n21 = Get l_curQualityToCheck() -> (l_curQualityToCheck=Low)
   AddError(
    execute=@n15.else
    NewError=Append!(
     A="MinLOD for "
     B=Enum to String(Enumerator=@n21)
     C=" is "
     D="To String (Integer)"(InInt=@n19)
     E=" but the mesh only has "
     F="To String (Integer)"(InInt=@n20)
     G:String=" LODs.  Reduce the index or add LODs so it's valid."
    ) -> (ReturnValue)
    Depth=1
   )
   > SetSeverity(NewSeverity=NewEnumerator2, OnlyIfHigher=true)
   Set l_vertCount(execute=@n14.else, l_vertCount="Get (a ref)"(Array=Get l_vertsPerLOD(), Dimension 1=0)) -> (Output_Get=0)
   > Branch(Condition="integer > integer"(A=Get l_vertCount() -> (l_vertCount=0), B=Get l_vertCountMax() -> (l_vertCountMax=0)))
   n22 = Get l_vertCount() -> (l_vertCount=0)
   n23 = Get l_vertCountMax() -> (l_vertCountMax=0)
   n24 = Get l_curQualityToCheck() -> (l_curQualityToCheck=Low)
   > AddError(
    NewError=Append!(
     A="No MinLOD is assigned to "
     B=Enum to String(Enumerator=@n24)
     C=" and LOD0 is too many verts: "
     D="To String (Integer)"(InInt=@n22)
     E=" and should be less than: "
     F="To String (Integer)"(InInt=@n23)
    ) -> (ReturnValue)
    Depth=1
   )
   > SetSeverity(NewSeverity=NewEnumerator2, OnlyIfHigher=true)
   Return Node!(
    execute=@n5.then_2
    Valid:bool="Equal (Enum)"(
     A:Enum_GART_ReportSeverity Enum=GetCurrentSeverity!() -> (CurrentSeverity:Enum_GART_ReportSeverity Enum)
     B:Enum_GART_ReportSeverity Enum=NewEnumerator4
    )
    ReportData:Struct GART Report Data Structure=Make Struct GART Report Data(
     DisplayTitle_20_#1=LOD
     DisplayDescShort_29_#2=MinLOD Vert Count
     Severity_46_#3=NewEnumerator0
    )
   )
  ]}
 NoLoad_MinLODCheck {nodes [
   n1 = NoLoad_MinLODCheck!() -> (then, AssetData:Asset Data Structure) {is=title K2Node_FunctionEntry}
   > Set l_assetData(l_assetData=@n1)
   > n2 = GetEditorWorld(self=Get UnrealEditorSubsystem())
   > Set l_world(l_world=@n2)
   > Set l_valid(l_valid=true)
   > n3 = Sequence()
   > AddError(NewError=Missing required MinLOD Quality Levels:)
   n4 = Get l_world()
   > n5 = NoLoad_GetMinLODLevelFromAssetData!(execute, AssetData:Asset Data Structure=Get l_assetData(), __WorldContext:Object Reference=@n4) -> (then, QualitiesInUse:Map of EPerQualityLevels Enums to Integers=Low)
   > Set l_meshQualityLevel(l_meshQualityLevel:Map of EPerQualityLevels Enums to Integers=@n5) -> (Output_Get:Map of EPerQualityLevels Enums to Integers=Low)
   > n6 = NoLoad_GetStaticMeshVertCountFromAssetData!(execute, AssetData:Asset Data Structure=Get l_assetData(), __WorldContext:Object Reference=@n4) -> (then, VertCount:int=0)
   > Set l_vertCount(l_vertCount=@n6) -> (Output_Get=0)
   n7 = Get QualityLevelMaxVerts() -> (QualityLevelMaxVerts=Low)
   n8 = Keys(execute=@n3.then_1, TargetMap=@n7)
   > n9 = Values(TargetMap=@n7)
   > n10 = For Each Loop(Array=@n8)
   > n11 = Set l_vertCountMax(l_vertCountMax="Get (a ref)"(Array=@n9, Dimension 1=@n10.Array Index)) -> (Output_Get=0)
   > Branch(Condition="integer > integer"(A=Get l_vertCount() -> (l_vertCount=0), B=@n11))
   n12 = Find(TargetMap=Get l_meshQualityLevel() -> (l_meshQualityLevel:Map of EPerQualityLevels Enums to Integers=Low), Key=@n10)
   > n13 = Branch(Condition=@n12.ReturnValue)
   > n14 = Branch(Condition="integer > integer"(A=@n12))
   n15 = Get l_vertCount() -> (l_vertCount=0)
   n16 = Get l_vertCountMax() -> (l_vertCountMax=0)
   n17 = AddError(
    execute=@n14.else
    NewError=Append!(
     A=Enum to String(Enumerator=@n10)
     B=" is "
     C="To String (Integer)"(InInt=@n15)
     D=" Verts and should be less than: "
     E="To String (Integer)"(InInt=@n16)
    ) -> (ReturnValue)
    Depth=1
   )
   Set l_valid(execute=[@n17.then, @n20.then], l_valid=false)
   n18 = Get l_vertCount() -> (l_vertCount=0)
   n19 = Get l_vertCountMax() -> (l_vertCountMax=0)
   n20 = AddError(
    execute=@n13.else
    NewError=Append(
     A=Enum to String(Enumerator=@n10)
     B=" has no MinLOD and LOD0 is too many verts. It has "
     C="To String (Integer)"(InInt=@n18)
     D=" and should be less than: "
     E="To String (Integer)"(InInt=@n19)
    )
    Depth=1
   )
   Return Node(
    execute=@n3.then_2
    Valid=Get l_valid() -> (l_valid:bool)
    ReportData=Make Struct GART Report Data(DisplayTitle_20_#1=LOD, DisplayDescShort_29_#2=MinLOD Vert Count)
   )
  ]}
 RunFilterLogicOnAssetData {nodes [
   n1 = RunFilterLogicOnAssetData()
   > n2 = Branch(Condition=@n1.AllowLoading)
   > n3 = Load_MinLODCheck!(execute, self, AssetData:Asset Data Structure=@n1) -> (then, Valid:bool, ReportData:Struct GART Report Data Structure) {is=title K2Node_CallFunction}
   > Return Node(Valid=@n3, ReportData=@n3.ReportData)
   n4 = NoLoad_MinLODCheck!(execute=@n2.else, self, AssetData:Asset Data Structure=@n1) -> (then, Valid:bool, ReportData:Struct GART Report Data Structure) {is=title K2Node_CallFunction}
   > Return Node(Valid=@n4, ReportData=@n4.ReportData)
  ]}
]
variables [QualityLevelMaxVerts {default {high=15000, low=1500, medium=5000}}]
