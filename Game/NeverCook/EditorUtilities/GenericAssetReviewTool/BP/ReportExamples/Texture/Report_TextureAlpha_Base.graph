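# Editor utility report for the Generic Asset Review Tool that flags Texture2D assets with useless or possibly useless alpha channels.
# RunFilterLogicOnAssetData calls CheckForEmptyAlpha, which needs loading allowed and screens out certain compression settings.
# DoesTextureHaveOptionalAlpha also requires source alpha below near-opaque.
# The source alpha min-max spread is then compared to MaxAlphaErrorTolerance, with near-zero spread getting a different severity than subtle variation.
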EditorUtilityBlueprint /Game/NeverCook/EditorUtilities/GenericAssetReviewTool/BP/ReportExamples/Texture/Report_TextureAlpha_Base
graphs [
 CheckForEmptyAlpha {nodes [
   n1 = CheckForEmptyAlpha!() -> (then, AssetData:Asset Data Structure, AllowLoad:bool=~) {is=title K2Node_FunctionEntry}
   > Set l_assetData(l_assetData=@n1)
   > n2 = Branch(Condition=@n1.AllowLoad)
   > n3 = Cast To Texture2D(Object=GetAsset(InAssetData=Get l_assetData()))
   > Set l_texture(l_texture:Texture 2D Object Reference=@n3) -> (Output_Get:Texture 2D Object Reference)
   > n4 = Sequence!(execute) -> (then_0, then_1, then_2:Exec)
   n5 = Get l_texture() -> (l_texture:Texture 2D Object Reference)
   n6 = Get CompressionSettings(self=@n5)
   > Branch(Condition=OR Boolean!(
     A=Get CompressionNoAlpha!(self:Texture Object Reference=@n5) -> (CompressionNoAlpha:bool)
     B="Equal (Enum)"(A:TextureCompressionSettings Enum=@n6, B:TextureCompressionSettings Enum=TC_VectorDisplacementmap)
     C:bool="Equal (Enum)"(A=@n6, B=TC_EditorIcon)
    ) -> (ReturnValue))
   > Return Node!(execute, Valid:bool, ReportData:Struct GART Report Data Structure)
   Branch(execute=@n4.then_1, Condition=DoesTextureHaveOptionalAlpha!(texture:Texture Object Reference=Get l_texture()) -> (ReturnValue:bool))
   > n7 = ComputeTextureSourceChannelMinMax!(execute, self:Texture Object Reference=Get l_texture()) -> (then, OutColorMin:Linear Color Structure, OutColorMax:Linear Color Structure, ReturnValue:bool)
   n8 = Break Linear Color!(LinearColor:"Linear Color Structure (by ref)"=@n7.OutColorMax) -> (R:float=~, G:float=~, B:float=~, A:float=~) {is=title K2Node_BreakStruct}
   n9 = Break Linear Color(LinearColor=@n7) -> (R=~, G=~, B=~, A=~)
   > Set l_difference(l_difference:double="Absolute (Float)"(A=float - float(A=@n9.A, B=@n8.A))) -> (Output_Get:double)
   > Branch(Condition="float <= float"(A=Get l_difference() -> (l_difference:double), B=Get MaxAlphaErrorTolerance() -> (MaxAlphaErrorTolerance:double)) -> (ReturnValue=~))
   > n10 = Branch(Condition="Nearly Equal (Float)"(A=Get l_difference(), B=0, ErrorTolerance=0.001))
   > AddError(NewError=Appears to have an empty alpha channel)
   > SetSeverity!(execute, self:GART Report Base Object Reference, NewSeverity:Enum_GART_ReportSeverity Enum=NewEnumerator2, OnlyIfHigher:bool=true) -> (then)
   > Set l_descShort(l_descShort:String=Useless Alpha Channel) -> (Output_Get:String)
   > Set l_valid(l_valid:bool=false) -> (Output_Get:bool)
   AddError(execute=@n10.else, NewError=Appears to have a very subtle alpha channel we could maybe do without.)
   > SetSeverity(NewSeverity=NewEnumerator1, OnlyIfHigher=true)
   > Set l_descShort(l_descShort=Possibly Useless Alpha Channel)
   > Set l_valid(l_valid=false)
   Return Node(
    execute=@n4.then_2
    Valid=Get l_valid() -> (l_valid:bool)
    ReportData=Make Struct GART Report Data(
     DisplayTitle_20_#1=Texture
     DisplayDescShort_29_#2=Get l_descShort() -> (l_descShort:String)
     Severity_46_#3=NewEnumerator0
    )
   )
   Return Node(execute=[@n3.CastFailed, @n2.else])
  ]}
 DoesTextureHaveOptionalAlpha {nodes [
   n1 = DoesTextureHaveOptionalAlpha!() -> (then, texture:Texture Object Reference) {is=title K2Node_FunctionEntry}
   > n2 = ComputeTextureSourceChannelMinMax(self=@n1)
   > n3 = Branch(Condition=@n2.ReturnValue)
   n4 = Get CompressionSettings(self=@n1)
   n5 = BreakColor(InColor=@n2)
   > Return Node!(
    execute
    ReturnValue:bool=AND Boolean!(
     A="Not Equal (Enum)"!(A:TextureCompressionSettings Enum=@n4, B:TextureCompressionSettings Enum=TC_HDR) -> (ReturnValue:bool)
     B="Not Equal (Enum)"(A=@n4, B=TC_Grayscale)
     C:bool="Not Equal (Enum)"(A=@n4, B=TC_Displacementmap)
     D:bool="Not Equal (Enum)"(A=@n4, B=TC_DistanceFieldFont)
     E:bool="Not Equal (Enum)"(A=@n4, B=TC_HalfFloat)
     F:bool="Not Equal (Enum)"(A=@n4, B=TC_Normalmap)
     G:bool="Not Equal (Enum)"(A=@n4, B=TC_HDR_Compressed)
     H:bool="Not Equal (Enum)"(A=@n4, B=TC_VectorDisplacementmap)
     I:bool="float < float"(A=@n5.A, B=0.999)
    ) -> (ReturnValue)
   )
   Return Node(execute=@n3.else, ReturnValue=false)
  ]}
 EventGraph {links [], nodes []}
 RunFilterLogicOnAssetData {nodes [
   n1 = RunFilterLogicOnAssetData()
   > n2 = CheckForEmptyAlpha!(execute, self, AssetData:Asset Data Structure=@n1, AllowLoad:bool=@n1.AllowLoading) -> (then, Valid:bool, ReportData:Struct GART Report Data Structure) {is=title K2Node_CallFunction}
   > Return Node!(execute, Valid:bool=@n2, ReportData:Struct GART Report Data Structure=@n2.ReportData)
  ]}
]
variables [MaxAlphaErrorTolerance {type=number, default=0.1}]
