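# Editor-only Generic Asset Review Tool report base that flags textures whose pixel count exceeds the MaxTextureSizes area.
# PixelCountCheck reads Dimensions, LODBias and NeverStream from asset registry tags, avoiding a load.
# When loading is allowed, ApplyMaxTextureSizeToDimensions also applies the texture's MaxTextureSize, using integer division for the aspect ratio.
# Unless NeverStream is set, LODBias halves both sides per level before the comparison and report text are produced.
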EditorUtilityBlueprint /Game/NeverCook/EditorUtilities/GenericAssetReviewTool/BP/ReportExamples/Texture/Report_TextureSize_Base
graphs [
 ApplyMaxTextureSizeToDimensions {nodes [
   n1 = ApplyMaxTextureSizeToDimensions!() -> (then, SizeX:int, SizeY:int, MaxSize:int) {is=title K2Node_FunctionEntry}
   > n2 = Set l_sizeX(l_sizeX:int=@n1) -> (Output_Get:int=0)
   > Set l_returnX(l_returnX:int=@n2) -> (Output_Get:int=0)
   > n3 = Set l_sizeY(l_sizeY:int=@n1.SizeY) -> (Output_Get:int=0)
   > Set l_returnY(l_returnY:int=@n3) -> (Output_Get:int=0)
   > Set l_max(l_max:int=@n1.MaxSize) -> (Output_Get:int=0)
   > n4 = Sequence()
   > Branch(Condition="integer > integer"(A=Get l_max() -> (l_max:int=0)))
   n5 = Get l_sizeX() -> (l_sizeX:int=0)
   n6 = Get l_sizeY() -> (l_sizeY:int=0)
   > Branch(Condition="integer > integer"(A="Max (Integer)"(A=@n5, B=@n6), B=Get l_max() -> (l_max=0)))
   > n7 = Branch(Condition="integer > integer"(A=Get l_sizeX() -> (l_sizeX=0), B=Get l_sizeY() -> (l_sizeY=0)))
   > Set l_returnX(l_returnX=Get l_max() -> (l_max=0)) -> (Output_Get=0)
   n8 = Get l_sizeX() -> (l_sizeX=0)
   n9 = Get l_sizeY() -> (l_sizeY=0)
   > Set l_returnY(l_returnY=int * int(A=int / int(A=@n9, B=@n8), B=Get l_max() -> (l_max=0))) -> (Output_Get=0)
   n10 = Get l_sizeY() -> (l_sizeY=0)
   n11 = Get l_sizeX() -> (l_sizeX=0)
   Set l_returnX(execute=@n7.else, l_returnX=int * int(A=int / int(A=@n11, B=@n10), B=Get l_max() -> (l_max=0))) -> (Output_Get=0)
   > Set l_returnY(l_returnY=Get l_max() -> (l_max=0)) -> (Output_Get=0)
   Return Node!(execute=@n4.then_1, SizeX_Out:int=Get l_returnX() -> (l_returnX:int=0), SizeY_Out:int=Get l_returnY() -> (l_returnY:int=0))
  ]}
 EventGraph {links [], nodes []}
 PixelCountCheck {nodes [
   n1 = PixelCountCheck!() -> (then, AssetData:Asset Data Structure, AllowLoad:bool=~) {is=title K2Node_FunctionEntry}
   > Set l_assetData(l_assetData=@n1)
   > Set l_allowLoad(l_allowLoad:bool=@n1.AllowLoad) -> (Output_Get:bool)
   n2 = BreakVector2D(InVec=Get MaxTextureSizes() -> (MaxTextureSizes:Vector 2D Structure))
   > Set l_maxPixelSize(l_maxPixelSize:int=Floor(A=float * float(A=@n2, B=@n2.Y) {is=title K2Node_PromotableOperator}) {is=title K2Node_CallFunction}) -> (Output_Get:int=0)
   > n3 = Sequence!(execute) -> (then_0, then_1, then_2:Exec, then_3:Exec, then_4:Exec, then_5:Exec, then_6:Exec)
   n4 = GetTagValue(InAssetData=Get l_assetData(), InTagName=MakeLiteralName(Value=Dimensions))
   > n5 = Branch(Condition=@n4.ReturnValue)
   n6 = Split(SourceString=Replace(SourceString=@n4, From=","), InStr=x, SearchDir=FromStart)
   > Set l_sizeX(l_sizeX=String To Integer(InString=@n6)) -> (Output_Get=0)
   > Set l_sizeY(l_sizeY=String To Integer(InString=@n6.RightS)) -> (Output_Get=0)
   Return Node!(execute=@n5.else, Valid:bool, ReportData:Struct GART Report Data Structure)
   n7 = GetTagValue(InAssetData=Get l_assetData(), InTagName=MakeLiteralName(Value=LODBias))
   n8 = Branch(execute=@n3.then_1, Condition=@n7.ReturnValue)
   > Set l_LODBias(l_LODBias:int=String To Integer(InString=@n7)) -> (Output_Get:int=0)
   Return Node(execute=@n8.else)
   n9 = GetTagValue(InAssetData=Get l_assetData(), InTagName=MakeLiteralName(Value=NeverStream))
   n10 = Branch(execute=@n3.then_2, Condition=@n9.ReturnValue)
   > Set l_neverStream(l_neverStream:bool="Equal Exactly (String)"(A=@n9, B=True)) -> (Output_Get:bool)
   Return Node(execute=@n10.else)
   Branch(execute=@n3.then_3, Condition=Get l_allowLoad() -> (l_allowLoad:bool))
   > n11 = Cast To Texture2D(Object=GetAsset(InAssetData=Get l_assetData()))
   > Set l_texture(l_texture:Texture 2D Object Reference=@n11) -> (Output_Get:Texture 2D Object Reference)
   > n12 = ApplyMaxTextureSizeToDimensions!(
    execute
    self
    SizeX:int=Get l_sizeX() -> (l_sizeX=0)
    SizeY:int=Get l_sizeY() -> (l_sizeY=0)
    MaxSize:int=Get MaxTextureSize!(self:Texture Object Reference=Get l_texture() -> (l_texture:Texture 2D Object Reference)) -> (MaxTextureSize:int=0)
   ) -> (then, SizeX_Out:int=0, SizeY_Out:int=0) {is=title K2Node_CallFunction}
   > n13 = Set l_sizeX(l_sizeX=@n12) -> (Output_Get=0)
   > n14 = Set l_sizeY(l_sizeY=@n12.SizeY_Out) -> (Output_Get=0)
   > PrintString(InString=Append!(A="Post Texture Size: ", B="To String (Integer)"(InInt=@n13), C:String=x, D:String="To String (Integer)"(InInt=@n14)) -> (ReturnValue))
   n15 = Branch(execute=@n3.then_4, Condition=Get l_neverStream() -> (l_neverStream:bool))
   Branch(execute=@n15.else, Condition="integer > integer"(A=Get l_LODBias() -> (l_LODBias:int=0)))
   > While Loop!(execute:Exec, Condition:bool="integer > integer"(A=Get l_LODBias() -> (l_LODBias=0))) -> (LoopBody:Exec, Completed:Exec) {is=title K2Node_MacroInstance}
   > Set l_LODBias(l_LODBias=int - int(A=Get l_LODBias() -> (l_LODBias=0))) -> (Output_Get=0)
   > Set l_sizeX(l_sizeX=int / int(A=Get l_sizeX() -> (l_sizeX=0), B=2)) -> (Output_Get=0)
   > Set l_sizeY(l_sizeY=int / int(A=Get l_sizeY() -> (l_sizeY=0), B=2)) -> (Output_Get=0)
   n16 = Get l_sizeX() -> (l_sizeX=0)
   n17 = Get l_sizeY() -> (l_sizeY=0)
   Set l_valid(execute=@n3.then_5, l_valid:bool="integer >= integer"(A=Get l_maxPixelSize() -> (l_maxPixelSize:int=0), B=int * int(A=@n16, B=@n17))) -> (Output_Get:bool)
   n18 = BreakVector2D(InVec=Get MaxTextureSizes())
   n19 = "To String (Text)"(InText="To Text (Float)"(Value=@n18.Y, MinimumIntegralDigits=1, MaximumIntegralDigits=324, MinimumFractionalDigits=0, MaximumFractionalDigits=0))
   n20 = "To String (Text)"(InText="To Text (Float)"(Value=@n18, MinimumIntegralDigits=1, MaximumIntegralDigits=324, MinimumFractionalDigits=0, MaximumFractionalDigits=0))
   n21 = "To String (Integer)"(InInt=Get l_maxPixelSize() -> (l_maxPixelSize=0))
   n22 = Append!(A="To String (Integer)"(InInt=Get l_sizeX() -> (l_sizeX=0)), B=x, C="To String (Integer)"(InInt=Get l_sizeY() -> (l_sizeY=0))) -> (ReturnValue)
   n23 = Select(
    Option 0:String="\r\nThis check does not include MaxTextureSize.  It's recommended for mipped textures to use LODBias instead of MaxTextureSize."
    Option 1:String
    Index=Get l_allowLoad()
   ) -> (ReturnValue:String)
   n24 = Append(A=@n20, B=x, C=@n19)
   Return Node(
    execute=@n3.then_6
    Valid=Get l_valid() -> (l_valid:bool)
    ReportData=Make Struct GART Report Data(
     DisplayTitle_20_#1=Textures
     DisplayDescShort_29_#2=Append(A=Too Large, B=" \r\n> ", C=@n24)
     DisplayDescLong_30_#3=Append!(
      A="Textures total pixel count is greater than "
      B=@n21
      C=.
      D:String=" A good example size is "
      E:String=@n20
      F:String=x
      G:String=@n19
      H:String=" but it is currently "
      I:String=@n22
      J:String=. Import a smaller texture or set an LODBias
      K:String=@n23
     ) -> (ReturnValue)
    )
   )
  ]}
 RunFilterLogicOnAssetData {nodes [
   n1 = RunFilterLogicOnAssetData()
   > n2 = PixelCountCheck!(execute, self, AssetData:Asset Data Structure=@n1, AllowLoad:bool=@n1.AllowLoading) -> (then, Valid:bool, ReportData:Struct GART Report Data Structure) {is=title K2Node_CallFunction}
   > Return Node(Valid=@n2, ReportData=@n2.ReportData)
  ]}
]
variables [MaxTextureSizes {type=Vector2D, default=<1024 1024>}]
