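# Editor utility Blueprint for the Generic Asset Review Tool whose ReviewAssets function reviews the selected assets.
# It runs their asset data through filters using a default list of static mesh and texture report classes.
# Holding Ctrl when the function runs allows asset loading during filtering.
# The resulting report map feeds a spawned EUW_GART_List_Main widget tab, which updates its list.
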EditorUtilityBlueprint /Game/NeverCook/EditorUtilities/GenericAssetReviewTool/BP/Actions/GART_AssetActionTest
graphs [
 EventGraph {links [], nodes []}
 ReviewAssets {nodes [
   ReviewAssets() {is=title K2Node_FunctionEntry}
   > n1 = GetSelectedAssetData()
   > n2 = GetEditorWorld(self=Get UnrealEditorSubsystem())
   > n3 = RunAssetsThroughFilters!(
    execute
    AssetDatas:Array of Asset Data Structures=@n1
    ReportClasses:Array of GART Report Base Soft Class References=Get ReportClasses() -> (ReportClasses:Array of GART Report Base Soft Class References)
    AllowLoading:bool=Is Control Down!(KeysState:"Modifier Keys State Structure (by ref)"=GetModifierKeysState!() -> (ReturnValue:Modifier Keys State Structure)) -> (ReturnValue:bool)
    __WorldContext:Object Reference=@n2
   ) -> (then, ReportDataMap:Map of Names to Struct GART Asset Report List Structures)
   > n4 = SpawnAndRegisterTabAndGetID!(
    execute
    self:Editor Utility Subsystem Object Reference=Get EditorUtilitySubsystem()
    InBlueprint:Editor Utility Widget Blueprint Object Reference=/Game/NeverCook/EditorUtilities/GenericAssetReviewTool/Widgets/EUW_GART_List_Main
   ) -> (then, NewTabID:Name=None, ReturnValue:Editor Utility Widget Object Reference)
   > n5 = Cast To EUW_GART_List_Main!(execute, Object=@n4.ReturnValue) -> (then, CastFailed, AsEUW GART List Main:EUW GART List Main Object Reference) {is=title K2Node_DynamicCast}
   > SetAssetReportData!(
    execute
    self:EUW GART List Main Object Reference=@n5
    ReportData:Map of Names to Struct GART Asset Report List Structures=@n3
    UpdateList:bool=true
   ) -> (then)
  ]}
]
variables [ReportClasses {
  type="array<Class>"
  default [
   /Game/NeverCook/EditorUtilities/GenericAssetReviewTool/BP/ReportExamples/StaticMesh/Report_StaticMeshVertCount_Base.Report_StaticMeshVertCount_Base_C
   /Game/NeverCook/EditorUtilities/GenericAssetReviewTool/BP/ReportExamples/StaticMesh/Report_StaticMeshCollision_Base.Report_StaticMeshCollision_Base_C
   /Game/NeverCook/EditorUtilities/GenericAssetReviewTool/BP/ReportExamples/Texture/Report_TextureSize_Base.Report_TextureSize_Base_C
   /Game/NeverCook/EditorUtilities/GenericAssetReviewTool/BP/ReportExamples/StaticMesh/Report_StaticMeshNanite_Base.Report_StaticMeshNanite_Base_C
   /Game/NeverCook/EditorUtilities/GenericAssetReviewTool/BP/ReportExamples/Texture/Report_TextureAlpha_Base.Report_TextureAlpha_Base_C
   /Game/NeverCook/EditorUtilities/GenericAssetReviewTool/BP/ReportExamples/StaticMesh/Report_StaticMeshCollisionLOD.Report_StaticMeshCollisionLOD_C
  ]
 }]
